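# Sound cue for effort vocals in the Athena BigFootWalk emote.
# A Random node picks among seven equally weighted effort wave variations, with branch culling exclusion set on that node.
# It uses emote attenuation, the emote music sound class and concurrency set, and a slight volume boost.
# A submix send is configured to the music audio-analysis submix.
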SoundCue /Game/Athena/Sounds/Emotes/BigFootWalk/Athena_Emote_BigFootWalk_Effort_Cue
root=Random_1
nodes [
 {
  id=Random_1
  children [WavePlayer_1,WavePlayer_3,WavePlayer_4,WavePlayer_5,WavePlayer_6,WavePlayer_7,WavePlayer_8]
  properties {bShouldExcludeFromBranchCulling=true, weights [1,1,1,1,1,1,1]}
 }
 {id=WavePlayer_1, properties {soundWaveAssetPtr=&/Emote_BigFootWalk_Effort_04}}
 {id=WavePlayer_3, properties {soundWaveAssetPtr=&/Emote_BigFootWalk_Effort_07}}
 {id=WavePlayer_4, properties {soundWaveAssetPtr=&/Emote_BigFootWalk_Effort_01}}
 {id=WavePlayer_5, properties {soundWaveAssetPtr=&/Emote_BigFootWalk_Effort_02}}
 {id=WavePlayer_6, properties {soundWaveAssetPtr=&/Emote_BigFootWalk_Effort_03}}
 {id=WavePlayer_7, properties {soundWaveAssetPtr=&/Emote_BigFootWalk_Effort_05}}
 {id=WavePlayer_8, properties {soundWaveAssetPtr=&/Emote_BigFootWalk_Effort_06}}
]
properties {
 attenuationSettings=&Fort_Emotes_Athena_Att
 maxDistance=1600
 soundClassObject=&Fort_Music_Emote
 volumeMultiplier=1.2
 concurrencyOverrides {maxCount=3, resolutionRule=StopOldest}
 concurrencySet [/Game/Sounds/Concurrency_Rules/Athena_EmoteMusic_Conc]
 soundSubmixSends [{minSendLevel=0, soundSubmix=&Fortnite_AudioAnalysis_Music}]
}
